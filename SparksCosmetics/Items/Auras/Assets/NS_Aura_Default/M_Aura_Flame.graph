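# Translucent unlit flame material for Niagara aura particles, colored by a curve atlas with time-of-day and Reactivity scaling on emissive.
# Panning noise carves feathered alpha and color masks from UV gradients, shaped by a noise-distorted half-sphere blob, with depth fade on opacity.
# World position offset combines spline thickening by FlameWidth, sine sway, and an optional clamped, UV-weighted pull driven by Niagara dynamic parameters and particle position.
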Material /SparksCosmetics/Items/Auras/Assets/NS_Aura_Default/M_Aura_Flame
graph {
 properties {
  bUsedWithNiagaraMeshParticles=true
  bUsedWithNiagaraSprites=true
  blendMode=BLEND_Translucent
  pixelDepthOffsetMode=PDOM_Legacy
  shadingModel=MSM_Unlit
  usedShadingModels=Unlit
 }
 nodes [
  n1 = NamedRerouteUsage()
  n2 = CurveAtlasRowParameter(CurveTime=@n1) {
   atlas=&/CurveAtlas_Aura_Fire
   curve=&/Curve_Aura_Fire_Blue
   parameterName=ColorCurve
  }
  n3 = MaterialFunctionCall(Value_Morning=Constant() {r=0.5}, Value_Day=Constant() {r=0.5}, Value_Evening=Constant() {r=0.25}, Value_Night=Constant() {r=0.2}) {materialFunction=&Scalar_Time_of_Day}
  n4 = Multiply(A=@n2, B=Saturate(None=Add(A=@n1, B=@n3.Result) {constB=0.1}))
  n5 = ScalarParameter() {defaultValue=3, group=Flame, parameterName=Intensity}
  MaterialFunctionCall(Emissive=Multiply(A=DynamicParameter() {defaultValue=<rgba -60 1 1 1>, paramNames [FlameWidth,Reactivity,Param3,Param4]}.Reactivity, B=Multiply(A=@n4, B=@n5))) {materialFunction=&MF_FX_EmissiveAdjustments}
  n6 = DynamicParameter() {desc=Parameters set by Niagara, parameterIndex=3, defaultValue=<rgba 0 1 1 1>, paramNames [X,Y,Z,Intensity]}
  n7 = MaterialFunctionCall(Float3=Subtract(
    A=MaterialFunctionCall(X=@n6.X, Y=@n6.X, Z=@n6.X) {materialFunction=&MakeFloat3}.Result
    B=ParticlePositionWS()
   )) {materialFunction=&BreakOutFloat3Components}
  n8 = ScalarParameter() {defaultValue=100, desc=Interpolate Clamp XY, group=WPO, parameterName=Interpolate XY Clamp, sortPriority=3}
  n9 = Clamp(None=AppendVector(A=@n7.R, B=@n7.R), Min=Multiply(A=@n8) {constB=-1}, Max=@n8) {maxDefault=1000, minDefault=-1000}
  n10 = ScalarParameter() {defaultValue=10, desc=Interpolate Clamp Z, group=WPO, parameterName=Interpolate Clamp Z, sortPriority=4}
  n11 = Clamp(None=@n7.B, Min=Multiply(A=@n10) {constB=-1}, Max=@n10) {minDefault=-1}
  n12 = TextureCoordinate()
  n13 = Power(
   Base=ComponentMask(None=OneMinus(None=@n12)) {g=true}
   Exp=ScalarParameter() {defaultValue=5, desc=Interpolate Power, group=WPO, parameterName=Interpolate Power, sortPriority=2}
  )
  n14 = StaticSwitchParameter(True=LinearInterpolate(B=AppendVector(A=@n9, B=@n11), Alpha=Multiply(A=@n6.Intensity, B=@n13)), False=Constant()) {defaultValue=true, group=WPO, parameterName=Use Interpolate WPO, sortPriority=1}
  n15 = DynamicParameter() {defaultValue=<rgba -60 1 1 1>, paramNames [FlameWidth,Reactivity,Param3,Param4]}
  n16 = MaterialFunctionCall(WidthBase=@n15.FlameWidth, WidthTip=@n15.FlameWidth) {materialFunction=&SplineThicken}
  n17 = TextureCoordinate()
  n18 = Add(A=Time(), B=ParticleRandom())
  n19 = ScalarParameter() {defaultValue=1, group=WPO, parameterName=WPOSineSpeed}
  n20 = Multiply(
   A=Add(A=ComponentMask(None=@n17) {g=true}, B=Multiply(A=@n18, B=@n19))
   B=ScalarParameter() {defaultValue=1, group=WPO, parameterName=WPOSineAmount}
  )
  n21 = Multiply(
   A=Multiply(A=Sine(None=@n20), B=Constant3Vector() {constant=<rgba 0.75 0.75 0 1>})
   B=ScalarParameter() {defaultValue=5, group=WPO, parameterName=WPOSineIntensity}
  )
  n22 = OneMinus(None=ComponentMask(None=TextureCoordinate()) {g=true})
  NamedRerouteDeclaration(None=Add(A=Add(A=@n14, B=@n16.WorldPositionOffset), B=Multiply(A=@n21, B=@n22))) {name=WPO, nodeColor=<rgba 1 0.8203 0 1>}
  n23 = TextureCoordinate()
  n24 = ComponentMask(None=TextureCoordinate()) {g=true}
  n25 = TextureCoordinate() {uTiling=1.3, vTiling=0.3}
  n26 = AppendVector(A=Constant() {r=2}, B=Constant() {r=3})
  n27 = Panner(Coordinate=Add(A=Multiply(A=@n25, B=@n26), B=ParticleRandom()), Speed=AppendVector(A=Constant(), B=Constant() {r=0.5})) {speedY=1}
  n28 = Multiply(A=LinearInterpolate(Alpha=TextureSample(UVs=@n27) {samplerType=SAMPLERTYPE_LinearColor, texture=&T_NoisePack_LinearColor}.R) {constA=-1, constB=0.5}) {constB=0.1}
  n29 = MaterialFunctionCall(UVs=AppendVector(A=ComponentMask(None=@n23) {r=true}, B=Add(A=@n24, B=@n28)), Texture Scale=Constant2Vector() {g=1.75, r=0.5}) {materialFunction=&ScaleUVsByCenter}
  n30 = Constant2Vector() {g=0.1}
  NamedRerouteDeclaration(None=TextureSample(UVs=Add(A=@n29.UVs, B=@n30) {constB=0}) {samplerSource=SSM_Clamp_WorldGroupSettings, texture=/Game/Effects/Fort_Effects/Textures/Mask/T_HalfSphere_D}.R) {name=BlobAlpha, nodeColor=<rgba 0.6875 0 1 1>}
  n31 = ComponentMask(None=TextureCoordinate()) {r=true}
  n32 = Power(Base=Sine(None=Multiply(A=@n31) {constB=0.5})) {constExponent=10}
  n33 = TextureCoordinate()
  n34 = Multiply(
   A=TextureCoordinate()
   B=AppendVector(
    A=ScalarParameter() {defaultValue=1.5, group=Flame, parameterName=FlameTextureXScale, sortPriority=0}
    B=ScalarParameter() {defaultValue=0.6, group=Flame, parameterName=FlameTextureYScale, sortPriority=1}
   )
  )
  n35 = ParticleRandom()
  n36 = Time()
  n37 = ParticleRandom()
  n38 = Constant()
  n39 = ScalarParameter() {defaultValue=1.25, group=Flame, parameterName=FlameTextureSpeed, sortPriority=3}
  n40 = TextureSampleParameter2D(UVs=Panner(Coordinate=Add(A=@n34, B=@n35), Time=Add(A=@n36, B=@n37), Speed=AppendVector(A=@n38, B=@n39)) {speedY=0.75}) {
   group=Flame
   parameterName=FlameTexture
   samplerType=SAMPLERTYPE_LinearGrayscale
   sortPriority=8
   texture=/SparksCharacterCommon/Items/Auras/Textures/T_Mojo_Fire_Noise
  }
  n41 = Multiply(A=Add(A=Multiply(A=@n32) {constB=0.3}, B=ComponentMask(None=@n33) {g=true}), B=OneMinus(None=@n40.G))
  n42 = ScalarParameter() {defaultValue=0.425, group=Flame, parameterName=FlameAlphaEdgeSubtract, sortPriority=4}
  n43 = Saturate(None=Divide(
    A=Subtract(A=@n41, B=@n42) {constB=0.3}
    B=ScalarParameter() {defaultValue=0.1, group=Flame, parameterName=FlameAlphaEdgeFeather, sortPriority=5}
   ) {constB=0.25})
  n44 = NamedRerouteUsage()
  NamedRerouteDeclaration(None=DepthFade(Opacity=Multiply(A=@n43, B=@n44)) {fadeDistanceDefault=50}) {name=FlameAlpha, nodeColor=<rgba 0 0.6953 1 1>}
  n45 = Divide(
   A=Subtract(A=@n41, B=ScalarParameter() {defaultValue=0.4, group=Flame, parameterName=FlameColorEdgeSubtract, sortPriority=6}) {constB=0.3}
   B=ScalarParameter() {defaultValue=0.175, group=Flame, parameterName=FlameColorEdgeFeather, sortPriority=7}
  ) {constB=0.25}
  NamedRerouteDeclaration(None=Multiply(A=@n44, B=Saturate(None=@n45))) {name=FlameEmissive, nodeColor=<rgba 0 0.3437 1 1>}
  NamedRerouteUsage()
  NamedRerouteUsage()
  ScalarParameter() {group=Flame, parameterName=FlameTextureHorizPos, sortPriority=2}
 ]
}
